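; Golden Spring Egg gatherable actor for the SpringEvent that drops gold piles when a player collects it.
; Remote instances play an egg-laying gameplay cue on spawn.
; The first interaction tosses ScalableFloat-sized gold pickups, sends a tagged quest stat event and sets bHasBeenCollected, whose RepNotify handles cosmetic claim VFX and pickup audio.
; An interaction-permission override allows second interactions only when the gather-enabled row is true.
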(event EventBeginPlay
  ;[0.1816 1 0.1744] Event BeginPlay
  ; Play Gameplay cue on spawn
  (Networking|SwitchHasAuthority
    (:Remote
      (GameplayCue|ExecuteGameplayCue%28Burst%29Local self (Variables|Default|GetGCN%20Egg%20Laying) (Ability|GameplayCue|MakeGameplayCueParameters 0.0)))))

(event Interaction|EventOnGatherOrInteract
  (params (InteractingPawn :type Fort%20Pawn%20Object%20Reference) (InteractionBeingAttempted :type EInteractionBeingAttempted%20Enum :default "FirstInteraction"))
  (bind _returnvalue (Fortnite|GetFortPlayerControllerfromActor InteractingPawn))
  (bind _returnvalue_2 (Quests|GetQuestManager _returnvalue "Invalid"))
  ;[0 0.08456 1] Event OnGatherOrInteract
  ; Give Gold in drop piles, replicate to client for VFX, and send quest event
  (|CallParentFunction InteractingPawn InteractionBeingAttempted)
  (switch Utilities|FlowControl|Switch|SwitchonEInteractionBeingAttempted InteractionBeingAttempted
    (:FirstInteraction
      ;;[0 0.08456 1] Give Gold rewards in smaller piles, if desired
      (for _ (range 1 (+ (Math|Float|Floor (ScalableFloat|GetValueatLevel (Variables|Default|GetSF_NumberOfGoldPiles))) 1))
        (Inventory|SpawnPickupInWorld "/Game/Items/ResourcePickups/Athena_WadsItemData.Athena_WadsItemData" (Math|Float|Floor (ScalableFloat|GetValueatLevel (Variables|Default|GetSF_AmoutGoldToDrop))) (Transformation|GetActorLocation self) "0.000000,0.000000,1.000000" 0 true true false -1 "Tossed" "Unset" 0 false))
      (Utilities|FlowControl|Sequence
        (:then_0
          (Utilities|IsValid _returnvalue
            (:Is%20Valid
              (Utilities|IsValid _returnvalue_2
                (:Is%20Valid
                  (Quest|SendComplexCustomStatEvent _returnvalue_2 0 (Variables|Default|GetQuestSourceTags) (Variables|Default|GetQuestTargetTags) 1))))))
        (:then_1
          ;;[0.3283 0 1] [1]: Replicate collection for FX
          (Variables|Default|SetbHasBeenCollected true))))))

(fn OnInteractAllowedToConsume
  (params (InteractingPawn :type Fort%20Pawn%20Object%20Reference) (InteractionBeingAttempted :type EInteractionBeingAttempted%20Enum :default "FirstInteraction"))
  (Utilities|IsValid InteractingPawn
    (:Is%20Valid
      ;[1 1 1] If second interact allow gather if enabled
      (switch Utilities|FlowControl|Switch|SwitchonEInteractionBeingAttempted InteractionBeingAttempted
        (:FirstInteraction
          (return (|CallParentFunction InteractingPawn "FirstInteraction" "IT_Simple")))
        (:SecondInteraction
          (return (Class|AthenaFunctionLibrary|RowtoBool (Variables|Gather|GetRow_GatherEnabled))))))
    (:Is%20Not%20Valid
      (return false))))

(fn OnRep_bHasBeenCollected
  (bind _self self)
  (bind _returnvalue (Transformation|GetActorLocation _self))
  (Networking|CanExecuteCosmeticEvents
    (:True
      (if (Variables|Default|GetbHasBeenCollected)
        (Effects|Components|ParticleSystem|SpawnFXAtLocation :FXSystemAsset (Variables|Default|GetVFX_EggClaimed) :Location _returnvalue :Scale "0.650000,0.650000,0.650000" :bAutoDestroy true :PoolingMethod "None" :bAutoActivate true)
        (Audio|PlaySoundatLocation "/SpringEvent/Items/Audio/SpringEggs_Pickup_Cue.SpringEggs_Pickup_Cue" _returnvalue)))))

(fn ConstructionScript
  (|CallParentFunction))
